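; Art-tool actor that generates an inverse color LUT from a scene-captured identity texture, with test-image preview planes.
; Construction script sizes planes and render target, builds dynamic materials and captures.
; Numbered functions, chained by Process, export the capture to EXR and run Python console commands to create cubes, apply the inverse and update a volume texture; the import step only prints paths and sets a material parameter.
; Extra functions extract identity cubes and visualize LUTs.
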(event EventBeginPlay)

(event Collision|EventActorBeginOverlap
  (params (OtherActor :type Actor%20Object%20Reference)))

(event EventTick
  (params (DeltaSeconds :type Float%20%28single-precision%29)))

(event Custom|ExportRT
  (params (TextureRenderTarget :type Texture%20Render%20Target%20Object%20Reference))
  (bind _prefix (Variables|Default|Getprefix))
  (bind _path_path (Utilities|Struct|BreakDirectoryPath (Variables|Getpath_Path)))
  (bind _returnvalue (Utilities|String|Append _path_path "/" _prefix "_" "identity" ".exr"))
  ;[1 1 1] export
  (bind _output_get (Variables|Default|SetSceneCaptureExportTextureFile _returnvalue))
  (Development|PrintString _output_get)
  (bind _structout (Utilities|Struct|SetmembersinImageWriteOptions (Utilities|Struct|MakeImageWriteOptions "EXR" (delegate Custom|ExportRTDelegate))))
  (Texture|ExporttoDisk TextureRenderTarget _returnvalue _structout)
  (Development|PrintString "Delegate"))

(event Custom|ExportRTDelegate
  (params (bSuccess :type Boolean)))

(fn 1%29UpdateRT
  (bind _scenecapturecomponent2d (Variables|Default|GetSceneCaptureComponent2D))
  (Variables|/Script/Engine.SceneCaptureComponent2D|SetTextureTarget (Variables|Default|GetSceneCaptureRenderTarget) _scenecapturecomponent2d)
  (Rendering|SceneCapture|CaptureScene _scenecapturecomponent2d :!class /Script/Engine.SceneCaptureComponent2D))

(fn 2%29ExportRTtoTexture
  (Default|1%29UpdateRT)
  (CallFunction|ExportRT (Variables|Default|GetSceneCaptureRenderTarget)))

(fn 3%29CreateCubes
  (bind _path (Variables|Default|Getpath))
  (bind _path_2 (Utilities|Struct|BreakDirectoryPath _path))
  (bind _rendertargetscale (Variables|Default|GetrenderTargetScale))
  (bind _returnvalue (Utilities|String|ToString(Float) _rendertargetscale))
  (bind _returnvalue_2 (Utilities|Paths|ConverttoAbsolutePath _path_2))
  (bind _scenecaptureexporttexturefile (Variables|Default|GetSceneCaptureExportTextureFile))
  (bind _returnvalue_3 (Utilities|String|Append "py TechArt\\LUTs\\createCubes.py " _scenecaptureexporttexturefile " " _returnvalue " " _returnvalue_2))
  (Development|PrintString _returnvalue_3)
  (Development|ExecuteConsoleCommand _returnvalue_3))

(fn 4%29ApplyInverseCubetoIdentity
  (bind _identity (Variables|Default|GetIdentity))
  (bind _returnvalue (Utilities|GetDisplayName _identity))
  (bind _returnvalue_2 (Utilities|Paths|GetSystemPath _identity))
  (bind _returnvalue_3 (Utilities|Paths|GetPath _returnvalue_2))
  (bind _path_path (Utilities|Struct|BreakDirectoryPath (Variables|Getpath_Path)))
  (bind _returnvalue_4 (Utilities|Paths|ConverttoAbsolutePath _path_path))
  (bind _returnvalue_5 (Utilities|String|Append _returnvalue_3 "/"))
  (bind _returnvalue_6 (Utilities|String|Append _returnvalue ".exr"))
  (bind _returnvalue_7 (Utilities|String|Append _returnvalue_5 _returnvalue_6 " "))
  (bind _prefix (Variables|Default|Getprefix))
  (bind _returnvalue_8 (Utilities|String|Append _returnvalue_4 "/" _prefix "_inverse.cube"))
  (bind _returnvalue_9 (Utilities|String|Append _returnvalue_4 "/" _prefix "_inverse.exr"))
  (bind _returnvalue_10 (Utilities|String|Append "py TechArt\\LUTs\\apply_lut.py " _returnvalue_8 " " _returnvalue_7 " " _returnvalue_9))
  (Development|PrintString _returnvalue_7)
  (Development|PrintString _returnvalue_10)
  (Development|ExecuteConsoleCommand _returnvalue_10))

(fn 5%29ImportInverseTexture
  (bind _path_path (Utilities|Struct|BreakDirectoryPath (Variables|Getpath_Path)))
  (bind (_lefts _rights) (Utilities|String|Split _path_path "Content"))
  (Development|PrintString (Utilities|String|Append (Utilities|GetObjectPathString (Variables|Default|GetinvertedTexture)) " " (Utilities|String|Append "/Game" _rights "/")))
  (Development|PrintString (Utilities|String|Append (Utilities|Paths|ConverttoAbsolutePath _path_path) "/" (Variables|Default|Getprefix) "_inverse.exr"))
  (Rendering|Material|SetScalarParameterValue (Variables|Default|GetDMI_testImage_invertedLUT) "IdentityMult" (Variables|Default|GetmaterialMult) :!class /Script/Engine.MaterialInstanceDynamic))

(fn 6%29UpdateVolumeTexture
  (bind _invertedtexture (Variables|Default|GetinvertedTexture))
  (bind _returnvalue (Utilities|GetObjectPathString _invertedtexture))
  (bind _volumetexture (Variables|Default|GetvolumeTexture))
  (bind _returnvalue_2 (Utilities|GetObjectPathString _volumetexture))
  (bind _size (Class|VolumetricsBPLIbrary|GetTextureSize (Variables|Default|GetinvertedTexture)))
  (bind (_size_x _size_y) (Utilities|Struct|BreakIntPoint _size))
  (Development|PrintString (Utilities|String|Append "py TechArt\\LUTs\\updateVolumetexture.py " " " _returnvalue_2 " " _returnvalue " " (Utilities|String|ToString(Integer) _size_y)))
  (Development|ExecuteConsoleCommand (Utilities|String|Append "py TechArt\\LUTs\\updateVolumetexture.py " " " _returnvalue_2 " " _returnvalue " " (Utilities|String|ToString(Integer) _size_y))))

(fn 9%29CreateIdentityCubes
  (bind _path (Variables|Default|Getpath))
  (bind _path_2 (Utilities|Struct|BreakDirectoryPath _path))
  (bind _returnvalue (Utilities|String|ToString(Float) 1.0))
  (bind _identitytexturefile_filepath (Utilities|Struct|BreakFilePath (Variables|GetIdentityTextureFile_FilePath)))
  (bind _returnvalue_2 (Utilities|String|Append "py TechArt\\LUTs\\extract_lut.py " _identitytexturefile_filepath " " _returnvalue " " _path_2))
  (Development|PrintString _returnvalue_2)
  (Development|ExecuteConsoleCommand _returnvalue_2))

(fn a%29visualize
  (bind _self self)
  (bind _returnvalue (Utilities|GetObjectPathString _self))
  (bind _cube_filepath (Utilities|Struct|BreakFilePath (Variables|Getcube_FilePath)))
  (bind _returnvalue_2 (Utilities|String|Append "py TechArt\\LUTs\\visualize_lut.py " _cube_filepath " " _returnvalue))
  (Components|InstancedStaticMesh|ClearInstances (Variables|Default|GetInstancedStaticMesh))
  (Development|PrintString _returnvalue_2)
  (Development|ExecuteConsoleCommand _returnvalue_2))

(fn b%29clearVisualization
  (Components|InstancedStaticMesh|ClearInstances (Variables|Default|GetInstancedStaticMesh)))

(fn MID
  (bind _returnvalue (Rendering|Material|CreateDynamicMaterialInstance (Variables|Default|GetIdentityPlane) 0 "/Game/ArtTools/Tech_Art/LUT/M_identity.M_identity" "None" :!class /Script/Engine.PrimitiveComponent))
  (Rendering|Material|SetTextureParameterValue _returnvalue "Texture" (Variables|Default|GetIdentity))
  (Rendering|Material|SetScalarParameterValue _returnvalue "IdentityMult" (Variables|Default|GetIdentityMult) :!class /Script/Engine.MaterialInstanceDynamic))

(fn Process
  (Default|1%29UpdateRT)
  (Default|2%29ExportRTToTexture)
  (Default|3%29CreateCubes)
  (Default|4%29ApplyInverseCubeToIdentity)
  (Default|5%29ImportInverseTexture)
  (Default|6%29UpdateVolumeTexture))

(fn ConstructionScript
  (bind _path (Variables|Default|Getpath))
  (bind _path_2 (Utilities|Struct|BreakDirectoryPath _path))
  (bind _returnvalue (Utilities|Paths|ConverttoRelativePath _path_2))
  (bind _plane_testimage (Variables|Default|GetPlane_testImage))
  (bind _scenecapturerendertarget (Variables|Default|GetSceneCaptureRenderTarget))
  (bind _plane_testimage_inverselut (Variables|Default|GetPlane_testImage_inverseLUT))
  (bind _identitypreviewscale (Variables|Default|GetidentityPreviewScale))
  (Variables|Default|Setpath (Utilities|Struct|MakeDirectoryPath _returnvalue))
  (Variables|Default|SetabsolutePath (Utilities|Struct|MakeDirectoryPath (Utilities|Paths|ConverttoAbsolutePath _returnvalue)))
  (Development|PrintString _returnvalue)
  ;[1 1 1] scale of identity
  (bind _size (Class|VolumetricsBPLIbrary|GetTextureSize (Variables|Default|GetIdentity)))
  (bind (_size_x _size_y) (Utilities|Struct|BreakIntPoint _size))
  (Transformation|SetWorldScale3D (Variables|Default|GetidentityTransform) (Math|Vector|MakeVector _identitypreviewscale (* (/ (Math|Conversions|ToFloat(Integer) _size_y) (Math|Conversions|ToFloat(Integer) _size_x)) _identitypreviewscale) 1.0))
  (Variables|/Script/Engine.SceneCaptureComponent2D|SetOrthoWidth (* _identitypreviewscale 1e+02) (Variables|Default|GetSceneCaptureComponent2D))
  (Rendering|ResizeRenderTarget2D (Variables|Default|GetSceneCaptureRenderTarget) (Math|Float|Truncate (* _size_x (Variables|Default|GetrenderTargetScale))) (Math|Float|Truncate (* _size_y (Variables|Default|GetrenderTargetScale))))
  (Variables|/Script/Engine.Texture|SetFilter "TF_Nearest" _scenecapturerendertarget)
  ;[1 1 1] create render target
  (Variables|/Script/Engine.Texture|SetLODGroup "TEXTUREGROUP_Pixels2D" _scenecapturerendertarget)
  ;[1 1 1] aspect ratio of test images
  (bind _size_2 (Class|VolumetricsBPLIbrary|GetTextureSize (Variables|Default|GetTestImage)))
  (bind (_size_x_2 _size_y_2) (Utilities|Struct|BreakIntPoint _size_2))
  (Transformation|SetWorldScale3D (Variables|Default|GetPlane_testImage) (Math|Vector|MakeVector 1.0 (/ (Math|Conversions|ToFloat(Integer) _size_y_2) (Math|Conversions|ToFloat(Integer) _size_x_2)) 1.0))
  (Transformation|SetWorldScale3D (Variables|Default|GetPlane_testImage_inverseLUT) (Math|Vector|MakeVector 1.0 (/ (Math|Conversions|ToFloat(Integer) _size_y_2) (Math|Conversions|ToFloat(Integer) _size_x_2)) 1.0))
  (Default|MID)
  ;[1 1 1] DMIs
  (bind _returnvalue_2 (Rendering|Material|CreateDynamicMaterialInstance _plane_testimage 0 "/Game/ArtTools/Tech_Art/LUT/M_LUT_testImage.M_LUT_testImage" "None" :!class /Script/Engine.PrimitiveComponent))
  (Variables|Default|SetDMI_testImage _returnvalue_2)
  (Rendering|Material|SetTextureParameterValue _returnvalue_2 "Texture" (Variables|Default|GetTestImage))
  (Rendering|Material|SetScalarParameterValue _returnvalue_2 "IdentityMult" (Variables|Default|GetIdentityMult) :!class /Script/Engine.MaterialInstanceDynamic)
  (Rendering|Material|SetMaterial _plane_testimage 0 _returnvalue_2)
  (bind _returnvalue_3 (Rendering|Material|CreateDynamicMaterialInstance _plane_testimage_inverselut 0 "/Game/ArtTools/Tech_Art/LUT/M_LUT_testImage_inverted.M_LUT_testImage_inverted" "None" :!class /Script/Engine.PrimitiveComponent))
  (Variables|Default|SetDMI_testImage_invertedLUT _returnvalue_3)
  (Rendering|Material|SetTextureParameterValue _returnvalue_3 "Texture" (Variables|Default|GetTestImage))
  (Rendering|Material|SetScalarParameterValue (Variables|Default|GetDMI_testImage_invertedLUT) "IdentityMult" (Variables|Default|GetmaterialMult) :!class /Script/Engine.MaterialInstanceDynamic)
  (Rendering|Material|SetMaterial _plane_testimage_inverselut 0 _returnvalue_3)
  (Default|1%29UpdateRT))
