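# M_LUT_testImage is an unlit material in the LUT tech art tools, named as a test image for LUT work.
# Its only node is a 2D texture sample parameter named Texture, which defaults to the T_colorWheel image.
# The graph shows no math or color processing, and no link from the sample to a material output.
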Material /Game/ArtTools/Tech_Art/LUT/M_LUT_testImage
graph {
 properties {pixelDepthOffsetMode=PDOM_Legacy, shadingModel=MSM_Unlit, usedShadingModels=Unlit}
 nodes [TextureSampleParameter2D() {parameterName=Texture, texture=&/T_colorWheel}]
}
