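# Material function that applies a volume-texture LUT, defaulting to an inverse LUT, to an input color.
# The "Flip GB" static switch, on by default, replicates red into all three channels before lookup.
# Coordinates get a half-texel scale and offset for a 22-texel LUT.
# The sampled color is divided by eye adaptation to cancel exposure, then scaled by the mult input.
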MaterialFunction /Game/ArtTools/Tech_Art/LUT/MF_applyLUT
graph {nodes [
  Constant() {desc=32 / 33, r=0.9697}
  Constant() {desc=0.5 / 33, r=0.0152}
  Divide() {constB=1.2}
  n1 = Divide() {constA=21, constB=22}
  n2 = FunctionInput()
  n3 = MaterialFunctionCall(Color=@n2) {materialFunction=&SplitComponents}
  n4 = StaticSwitchParameter(
   True=MaterialFunctionCall(X=@n3.R, Y=@n3.R, Z=@n3.R) {materialFunction=&MakeFloat3}.Result
   False=@n2
  ) {defaultValue=true, parameterName=Flip GB}
  n5 = TextureSampleParameterVolume!(UVs=Add(A=Multiply(A=@n4, B=@n1), B=Divide() {constA=0.5, constB=22}), Apply View MipBias) {parameterName=LUT, texture=&/VT_inverseLUT}
  n6 = Divide(B=EyeAdaptation().EyeAdaptation)
  FunctionOutput(None=Multiply(A=Multiply(A=@n5.RGB, B=@n6) {constB=0.6}, B=FunctionInput() {inputName=mult, inputType=FunctionInput_Scalar}))
 ]}
