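# Unlit test material in the Tech Art LUT tools that runs an image through a color lookup, named as the inverted variant.
# A texture parameter, defaulting to a color wheel, feeds its RGB into the MF_applyLUT material function, with an IdentityMult scalar as the multiplier input.
# A component mask keeps only the red channel of the function's result.
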Material /Game/ArtTools/Tech_Art/LUT/M_LUT_testImage_inverted
graph {
 properties {pixelDepthOffsetMode=PDOM_Legacy, shadingModel=MSM_Unlit, usedShadingModels=Unlit}
 nodes [ComponentMask(None=MaterialFunctionCall!(
    mult=ScalarParameter() {defaultValue=1, parameterName=IdentityMult}
    In=TextureSampleParameter2D() {parameterName=Texture, texture=&/T_colorWheel}.RGB
   ) {materialFunction=&/MF_applyLUT}.Result) {r=true}]
}
