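# M_colorCubes is a Default Lit material for instanced static meshes, in the LUT tech-art tools, that colors each instance from its own custom data.
# Three per-instance custom data floats pass through vertex interpolators to reach the pixel shader and combine via MakeFloat3 into one RGB vector.
# A Power node applies a fixed gamma-like exponent to that vector, alongside two standalone constants.
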Material /Game/ArtTools/Tech_Art/LUT/M_colorCubes
graph {
 properties {bUsedWithInstancedStaticMeshes=true, pixelDepthOffsetMode=PDOM_Legacy, usedShadingModels=Default Lit}
 nodes [
  Constant() {r=1}
  Constant()
  n1 = PerInstanceCustomData!(DefaultValue) {constDefaultValue=1}
  n2 = PerInstanceCustomData() {constDefaultValue=1, dataIndex=1}
  n3 = PerInstanceCustomData() {dataIndex=2}
  Power(Base=MaterialFunctionCall(X=VertexInterpolator(VS=@n1).PS, Y=VertexInterpolator(VS=@n2).PS, Z=VertexInterpolator(VS=@n3).PS) {materialFunction=&MakeFloat3}.Result) {constExponent=1.66}
 ]
}
