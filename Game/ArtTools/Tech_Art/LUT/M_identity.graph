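# M_identity is an unlit material in the tech-art LUT folder, built around the T_identityUE texture, apparently an identity LUT.
# A texture parameter's RGB is multiplied by the IdentityMult scalar parameter.
# A separate SphereMask compares UVs to the red and green channels of a vector parameter, forming a small radial mask.
# The graph shows no link between that mask and the multiply, nor which node drives the material output.
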Material /Game/ArtTools/Tech_Art/LUT/M_identity
graph {
 properties {pixelDepthOffsetMode=PDOM_Legacy, shadingModel=MSM_Unlit, usedShadingModels=Unlit}
 nodes [
  Multiply(
   A=TextureSampleParameter2D() {parameterName=Texture, texture=&/T_identityUE}.RGB
   B=ScalarParameter() {defaultValue=1, parameterName=IdentityMult}
  )
  SphereMask(A=TextureCoordinate(), B=ComponentMask(None=VectorParameter().RGB) {g=true, r=true}) {attenuationRadius=0.1}
 ]
}
